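# Material function that draws a triangle arrow pointer for the Specter UI reticle, placed on either the left or right according to a 0-or-1 scalar input.
# That input lerps the horizontal UV offset and rotation between mirrored values, with a fixed vertical offset.
# The offset UVs pass through CustomRotator and a non-uniform ScaleUVsByCenter, then sample the arrow texture with clamped addressing, outputting saturated RGB.
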MaterialFunction /SpecterGameplay/UI/Materials/MF_UI_ReticleSpecter_ArrowPointer
graph {nodes [
  n1 = TextureCoordinate()
  n2 = FunctionInput() {description=Left Or Right 0 or 1, inputName=b_LeftRight, inputType=FunctionInput_Scalar}
  n3 = AppendVector(A=LinearInterpolate(A=Constant() {r=0.14}, B=Constant() {r=-0.14}, Alpha=@n2), B=Constant() {r=0.3565})
  n4 = Constant() {r=-0.95}
  n5 = Constant() {r=0.95}
  n6 = Constant() {r=0.15}
  n7 = Constant() {r=0.8}
  n8 = MaterialFunctionCall(
   UVs=MaterialFunctionCall(UVs=Add(A=@n1, B=@n3) {constB=0.2}, "Rotation Angle (0-1)"=LinearInterpolate(A=@n4, B=@n5, Alpha=@n2)) {materialFunction=&CustomRotator}.Rotated Values
   Texture Scale=Multiply(A=AppendVector(A=@n6, B=@n7), B=Constant() {r=0.08})
  ) {materialFunction=&ScaleUVsByCenter}
  FunctionOutput(None=Saturate(None=TextureSample(UVs=@n8.UVs) {samplerSource=SSM_Clamp_WorldGroupSettings, texture=/SpecterGameplay/UI/Textures/T_UI_Arrow_Triangle}.RGB)) {outputName=Saturate}
 ]}
