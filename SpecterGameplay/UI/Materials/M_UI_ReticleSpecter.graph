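# Translucent UI material that procedurally draws the Specter reticle, with a radial progress bar, outer line, arrows and a texture detail.
# Rounded-box outlines, masked in angular space, form the line arcs, and PbarProgress offsets a floored radial coordinate to fill the inner one.
# An anti-aliased distance-field detail and mirrored arrow-pointer functions add shapes.
# Named reroutes organise these into MainShape and parameter-tinted Color; several leftover branches feed nothing.
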Material /SpecterGameplay/UI/Materials/M_UI_ReticleSpecter
graph {
 properties {bAutomaticallySetUsageInEditor=false, blendMode=BLEND_Translucent, materialDomain=MD_UI, usedShadingModels=Default Lit}
 nodes [
  n1 = MaterialFunctionCall(UVs=NamedRerouteUsage(), "Rotation Angle (0-1)"=Constant() {r=0.192}) {materialFunction=&CustomRotator}
  n2 = Constant() {r=1}
  n3 = Constant() {r=-1}
  n4 = MaterialFunctionCall(Vector or UVs=Multiply(A=Subtract(A=@n1.Rotated Values) {constB=0.5}, B=AppendVector(A=@n2, B=@n3))) {materialFunction=&VectorToRadialValue}
  ComponentMask(None=Floor(None=Add(A=@n4.Radial Coordinates) {constB=0.88})) {r=true}
  n5 = MaterialFunctionCall(UVs=NamedRerouteUsage(), "Rotation Angle (0-1)"=Constant() {r=0.2}) {materialFunction=&CustomRotator}
  n6 = Constant() {r=1}
  n7 = Constant() {r=-1}
  n8 = MaterialFunctionCall(Vector or UVs=Multiply(A=Subtract(A=@n5.Rotated Values) {constB=0.5}, B=AppendVector(A=@n6, B=@n7))) {materialFunction=&VectorToRadialValue}
  ComponentMask(None=Floor(None=Add(A=@n8.Radial Coordinates) {constB=0.9})) {r=true}
  MaterialFunctionCall() {materialFunction=&Blend_Overlay}
  NamedRerouteDeclaration(None=TextureCoordinate()) {name=TexCoord, nodeColor=<rgba 1 0 0.0031 1>}
  n9 = Constant()
  n10 = Constant() {r=0.4}
  n11 = MaterialFunctionCall(UVs=Add(A=NamedRerouteUsage(), B=AppendVector(A=@n9, B=@n10)), Texture Scale=Constant() {r=0.2}) {materialFunction=&ScaleUVsByCenter}
  n12 = Saturate(None=MaterialFunctionCall(
    DistanceField=TextureSample(UVs=@n11.UVs) {samplerSource=SSM_Clamp_WorldGroupSettings, texture=/SpecterGameplay/UI/Textures/T_UI_ReticleSpecterMix}.R
    Threshold=Constant() {r=0.5}
   ) {materialFunction=&MF_UI_AntiAliasing}.Outer)
  n13 = SmoothStep(Value=OneMinus(None=MaterialFunctionCall() {materialFunction=&LinearGradient}.VGradient)) {constMax=0.95, constMin=0.8}
  n14 = LinearInterpolate(A=LinearInterpolate(B=@n12, Alpha=Saturate(None=@n13)), Alpha=NamedRerouteUsage())
  n15 = NamedRerouteUsage()
  NamedRerouteDeclaration(None=LinearInterpolate(A=LinearInterpolate(A=@n14, Alpha=@n15), Alpha=NamedRerouteUsage())) {name=MainShape, nodeColor=<rgba 0 1 0.6784 1>}
  n16 = ScalarParameter() {defaultValue=0.75, parameterName=InnerLineSize}
  n17 = MaterialFunctionCall(
   UVs=NamedRerouteUsage()
   Size=@n16
   CornerRoundness=@n16
   OutlineThickness=ScalarParameter() {defaultValue=0.025, parameterName=InnerLineThickness}
  ) {materialFunction=&MF_UI_Box}
  n18 = MaterialFunctionCall!(Vector or UVs) {materialFunction=/FortUILibrary/MaterialFunctions/Basics/MF_UI_CheaperVectorToRadial}
  n19 = AppendVector(A=Constant() {r=0.86}, B=Constant() {r=1})
  n20 = LinearInterpolate(
   B=@n17.Outline
   Alpha=Saturate(None=MaterialFunctionCall(UVs=@n18.Vector Converted to Angle, Size=@n19) {materialFunction=&MF_UI_Box}.Fill)
  )
  NamedRerouteDeclaration(None=LinearInterpolate(B=@n20, Alpha=ScalarParameter() {defaultValue=0.5, parameterName=PbarBackgroundOP})) {name=FillBackGround, nodeColor=<rgba 0 1 0.7422 1>}
  n21 = MaterialFunctionCall(UVs=NamedRerouteUsage(), "Rotation Angle (0-1)"=Constant() {r=0.25}) {materialFunction=&CustomRotator}
  n22 = Constant() {r=1}
  n23 = Constant() {r=1}
  n24 = MaterialFunctionCall(Vector or UVs=Multiply(A=Subtract(A=@n21.Rotated Values) {constB=0.5}, B=AppendVector(A=@n22, B=@n23))) {materialFunction=&VectorToRadialValue}
  n25 = LinearInterpolate(Alpha=ScalarParameter() {defaultValue=0.4, parameterName=PbarProgress}) {constA=0.05, constB=0.95}
  n26 = ComponentMask(None=Floor(None=Add(A=@n24.Radial Coordinates, B=@n25) {constB=0.9})) {r=true}
  NamedRerouteDeclaration(None=LinearInterpolate(B=@n20, Alpha=LinearInterpolate(B=@n26, Alpha=@n17.Outline))) {name=PbarFill, nodeColor=<rgba 0 1 0.7422 1>}
  n27 = ScalarParameter() {defaultValue=0.82, parameterName=OuterLineSIze}
  n28 = MaterialFunctionCall() {materialFunction=&MF_UI_CheaperVectorToRadial}
  n29 = AppendVector(A=Constant() {r=0.88}, B=Constant() {r=1})
  n30 = LinearInterpolate(
   B=MaterialFunctionCall(
    UVs=NamedRerouteUsage()
    Size=@n27
    CornerRoundness=@n27
    OutlineThickness=ScalarParameter() {defaultValue=0.01, parameterName=OuterLineThickness}
   ) {materialFunction=&MF_UI_Box}.Outline
   Alpha=Saturate(None=MaterialFunctionCall(UVs=@n28.Vector Converted to Angle, Size=@n29) {materialFunction=&MF_UI_Box}.Fill)
  )
  n31 = NamedRerouteUsage()
  NamedRerouteDeclaration(None=LinearInterpolate(B=LinearInterpolate(A=@n30, Alpha=@n31), Alpha=ScalarParameter() {defaultValue=0.9, parameterName=OutlineOP})) {name=OutLine, nodeColor=<rgba 0 1 0.7422 1>}
  n32 = VectorParameter() {parameterName=LineColour, defaultValue=<rgba 1 0.3389 0 1>}
  n33 = LinearInterpolate(A=VectorParameter() {parameterName=BG_Colour, defaultValue=<rgba 0.849 0.849 0.849 1>}.RGB, B=@n32.RGB, Alpha=NamedRerouteUsage())
  n34 = NamedRerouteUsage()
  n35 = LinearInterpolate(A=@n32.RGB, B=@n34, Alpha=ScalarParameter() {defaultValue=1, parameterName=b_PbarFull})
  n36 = NamedRerouteUsage()
  NamedRerouteDeclaration(None=LinearInterpolate(A=LinearInterpolate(A=@n33, B=@n35, Alpha=@n36), B=@n34, Alpha=NamedRerouteUsage())) {name=Color, nodeColor=<rgba 0 0.8338 1 1>}
  NamedRerouteDeclaration(None=@n12) {name=DetailMask, nodeColor=<rgba 0 1 0.5781 1>}
  n37 = VectorParameter() {parameterName=Fill_Colour, defaultValue=<rgba 0.4516 0.8229 0 1>}
  n38 = MaterialFunctionCall() {materialFunction=&LinearGradient}
  n39 = Constant3Vector() {constant=<rgba 0.2455 1 0.5631 1>}
  n40 = Saturate(None=MaterialFunctionCall!(Base=@n37.RGB, Blend=@n39) {materialFunction=/Engine/Functions/Engine_MaterialFunctions03/Blends/Blend_HardLight}.Result)
  NamedRerouteDeclaration(None=Saturate(None=LinearInterpolate(A=@n37.RGB, B=@n40, Alpha=@n38.VGradient))) {name=PbarColour, nodeColor=<rgba 0 0.9531 1 1>}
  n41 = Constant() {r=1}
  n42 = Constant()
  NamedRerouteDeclaration(None=LinearInterpolate(
    A=MaterialFunctionCall!(b_LeftRight=@n41) {materialFunction=&/MF_UI_ReticleSpecter_ArrowPointer}.Saturate
    Alpha=MaterialFunctionCall(b_LeftRight=@n42) {materialFunction=&MF_UI_ReticleSpecter_ArrowPointer}.Saturate
   )) {name=Arrows, nodeColor=<rgba 0.7812 1 0 1>}
  NamedRerouteUsage()
  NamedRerouteUsage()
  Reroute(None=@n39)
  n43 = TextureSample(UVs=MaterialFunctionCall(UVs=NamedRerouteUsage(), Texture Scale=Constant() {r=0.9}) {materialFunction=&ScaleUVsByCenter}.UVs) {samplerSource=SSM_Clamp_WorldGroupSettings, texture=&T_UI_ReticleSpecterMix}
  n44 = Constant() {r=0.3}
  Saturate(None=LinearInterpolate(
    B=MaterialFunctionCall(DistanceField=@n43.G, Threshold=@n44) {materialFunction=&MF_UI_AntiAliasing}.Outer
    Alpha=Constant() {r=0.4}
   ))
  n45 = Constant() {r=-0.14}
  n46 = Constant() {r=0.3565}
  n47 = MaterialFunctionCall(UVs=Add(A=NamedRerouteUsage(), B=AppendVector(A=@n45, B=@n46)) {constB=0.2}, "Rotation Angle (0-1)"=Constant() {r=0.95}) {materialFunction=&CustomRotator}
  n48 = Multiply(A=AppendVector(A=Constant() {r=0.15}, B=Constant() {r=0.8}), B=Constant() {r=0.08})
  Saturate(None=TextureSample(UVs=MaterialFunctionCall(UVs=@n47.Rotated Values, Texture Scale=@n48) {materialFunction=&ScaleUVsByCenter}.UVs) {samplerSource=SSM_Clamp_WorldGroupSettings, texture=/SpecterGameplay/UI/Textures/T_UI_Arrow_Triangle}.RGB)
  n49 = Add(A=NamedRerouteUsage(), B=AppendVector(A=Constant(), B=Constant() {r=0.5}))
  n50 = Constant() {r=0.25}
  n51 = Saturate(None=SmoothStep(Value=MaterialFunctionCall(RotationAngle=@n50) {materialFunction=&MF_UI_LinearGradient}.Result) {constMin=0.4})
  n52 = Multiply(A=@n51, B=Constant() {r=0.3}) {constB=2}
  n53 = Constant() {r=1}
  Saturate(None=MaterialFunctionCall(UVs=@n49, Size=AppendVector(A=@n52, B=@n53)) {materialFunction=&MF_UI_Box}.Fill)
  n54 = Multiply(A=@n51, B=Constant() {r=0.36}) {constB=2}
  n55 = Constant() {r=1}
  Saturate(None=MaterialFunctionCall(UVs=@n49, Size=AppendVector(A=@n54, B=@n55)) {materialFunction=&MF_UI_Box}.Fill)
 ]
}
