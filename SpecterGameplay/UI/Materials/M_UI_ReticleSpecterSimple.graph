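# Translucent UI material that draws the Specter reticle with a progress bar, outline and faint underlayer.
# MSDF textures, sampled with centre-scaled UVs, are converted to anti-aliased masks for the main shape, with optional vertical gradient fade, and the outline.
# A vertical threshold on PbarProgress builds the bar mask.
# Parameter-driven lerps, linked by named reroutes, blend main, progress, full-bar and outline colours and layer the masks' opacities.
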Material /SpecterGameplay/UI/Materials/M_UI_ReticleSpecterSimple
graph {
 properties {bAutomaticallySetUsageInEditor=false, blendMode=BLEND_Translucent, materialDomain=MD_UI, usedShadingModels=Default Lit}
 nodes [
  n1 = NamedRerouteUsage()
  n2 = ScalarParameter() {defaultValue=0.3, parameterName=Underlayer_OP}
  LinearInterpolate(
   A=LinearInterpolate(A=LinearInterpolate(B=@n1, Alpha=@n2), Alpha=NamedRerouteUsage())
   Alpha=SmoothStep(Value=NamedRerouteUsage()) {constMin=0.1}
  )
  n3 = MaterialFunctionCall!(MSDF RGB=TextureSample(UVs=NamedRerouteUsage(), Tex=TextureObjectParameter() {parameterName=TextureParameter, texture=/SpecterGameplay/UI/Textures/Reticle_Fill}) {samplerSource=SSM_Clamp_WorldGroupSettings, texture=/SpecterGameplay/UI/Textures/T_UI_ReticleSpecterMix}.RGB) {materialFunction=/FortUILibrary/MaterialFunctions/Shapes/Modifiers/MF_UI_MSDFtoSDF}
  MaterialFunctionCall!(
   DistanceField=@n3.SDF
   Threshold=ScalarParameter() {defaultValue=0.5, parameterName=ShapeAntiAlising}
   Thickness=ScalarParameter() {defaultValue=0.49, parameterName=ShapeThickness}
   Alignment
   Smoothing
  ) {materialFunction=/FortUILibrary/MaterialFunctions/Shapes/Modifiers/MF_UI_ShapeOutline}
  n4 = TextureCoordinate()
  n5 = AppendVector(A=Constant(), B=Constant())
  NamedRerouteDeclaration(None=MaterialFunctionCall(UVs=Add(A=@n4, B=@n5), Texture Scale=ScalarParameter() {defaultValue=1, parameterName=TextureScale}) {materialFunction=&ScaleUVsByCenter}.UVs) {name=TexCoord, nodeColor=<rgba 1 0 0.0031 1>}
  n6 = MaterialFunctionCall(
   DistanceField=@n3.SDF
   Threshold=ScalarParameter() {defaultValue=0.5, parameterName=ShapeAntiAlising}
   Smoothing=ScalarParameter() {parameterName=ShapeSmoothing}
  ) {materialFunction=&MF_UI_AntiAliasing}
  n7 = SmoothStep(
   Min=ScalarParameter() {parameterName=FadePosition}
   Value=OneMinus(None=MaterialFunctionCall() {materialFunction=&LinearGradient}.VGradient)
  ) {constMax=0.8}
  n8 = ScalarParameter() {defaultValue=1, parameterName=b_Fade}
  NamedRerouteDeclaration(None=LinearInterpolate(B=NamedRerouteDeclaration(None=@n6.Outer) {name=ShapeMask, nodeColor=<rgba 1 0.2344 0 1>}, Alpha=LinearInterpolate(B=@n7, Alpha=@n8))) {name=MainShape, nodeColor=<rgba 0 1 0.6784 1>}
  n9 = LinearInterpolate(
   A=VectorParameter() {parameterName=MainColour, defaultValue=<rgba 0.849 0.4773 0 1>}.RGB
   B=VectorParameter() {parameterName=Pbar_Colour, defaultValue=<rgba 0.2697 1 0 1>}.RGB
   Alpha=NamedRerouteUsage()
  )
  n10 = VectorParameter() {parameterName=Pbar_Full, defaultValue=<rgba 0.4516 0.8229 0 1>}
  n11 = ScalarParameter() {parameterName=b_PbarFull}
  n12 = LinearInterpolate(A=OneMinus(None=NamedRerouteUsage()), Alpha=NamedRerouteUsage())
  NamedRerouteDeclaration(None=LinearInterpolate(
    A=LinearInterpolate(A=@n9, B=@n10.RGB, Alpha=@n11)
    B=VectorParameter() {parameterName=OutlineColour, defaultValue=<rgba 1 1 1 1>}.RGB
    Alpha=SmoothStep(Value=@n12) {constMin=0.3}
   )) {name=Colour, nodeColor=<rgba 0.8047 0 1 1>}
  n13 = ScalarParameter() {defaultValue=0.3, parameterName=PbarProgress}
  n14 = OneMinus(None=ComponentMask(None=NamedRerouteUsage()) {g=true})
  n15 = OneMinus(None=SmoothStep(Min=@n13, Max=@n13, Value=SmoothStep(Value=@n14) {constMax=0.795, constMin=0.18}) {constMax=0.8})
  n16 = NamedRerouteUsage()
  NamedRerouteDeclaration(None=LinearInterpolate(B=LinearInterpolate(B=@n15, Alpha=@n16), Alpha=ScalarParameter() {defaultValue=1, parameterName=PbarOp})) {name=PBar, nodeColor=<rgba 1 0.0703 0 1>}
  NamedRerouteDeclaration(None=@n15) {name=PbarNoMask, nodeColor=<rgba 0.9453 0 1 1>}
  n17 = NamedRerouteUsage()
  n18 = Saturate(None=LinearInterpolate(B=NamedRerouteUsage(), Alpha=OneMinus(None=@n17)))
  NamedRerouteDeclaration(None=@n18) {name=OutlineMaskWithoutOp, nodeColor=<rgba 1 0 0.2578 1>}
  NamedRerouteDeclaration(None=LinearInterpolate(B=@n18, Alpha=ScalarParameter() {defaultValue=0.3, parameterName=OutlineOP})) {name=outlineMask, nodeColor=<rgba 0 0.1094 1 1>}
  n19 = TextureSample(UVs=NamedRerouteUsage()) {samplerSource=SSM_Clamp_WorldGroupSettings, texture=/SpecterGameplay/UI/Textures/Reticle_Layers}
  NamedRerouteDeclaration(None=MaterialFunctionCall(
    DistanceField=MaterialFunctionCall(MSDF RGB=@n19.RGB) {materialFunction=&MF_UI_MSDFtoSDF}.SDF
    Threshold=ScalarParameter() {defaultValue=0.2, parameterName=OutlineThreshold}
    Smoothing=ScalarParameter() {parameterName=OutLineSmoothing}
   ) {materialFunction=&MF_UI_AntiAliasing}.Outer) {name=OutlineShape, nodeColor=<rgba 0.4297 1 0 1>}
  NamedRerouteUsage()
 ]
}
